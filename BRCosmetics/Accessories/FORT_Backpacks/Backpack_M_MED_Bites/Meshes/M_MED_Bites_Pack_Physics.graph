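# Physics asset for the M_MED_Bites backpack cosmetic, providing box collision and joint limits for its dangling parts.
# Two angled boxes approximate the pack body, with single boxes on the handle and two right-side rope bones.
# The rope bones chain off the pack with locked twist and limited swing.
# The handle hinges from the pack on one swing axis, with twist and the other swing locked.
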PhysicsAsset /BRCosmetics/Accessories/FORT_Backpacks/Backpack_M_MED_Bites/Meshes/M_MED_Bites_Pack_Physics
bodies {
 dyn_handle_1_joint [{shapeType=Box, extentX=6.7949, extentY=10, extentZ=2, shapeName=dyn_handle_1_joint_box, center=<3.0357 0 1.0158>}]
 dyn_low_1_rope_r [{shapeType=Box, extentX=1.5715, extentY=0.9183, extentZ=0.5251, shapeName=dyn_low_1_rope_r_box, center=<0.6733 0.4969 0.2618>}]
 dyn_low_2_rope_r [{
   shapeType=Box
   extentX=2.8018
   extentY=0.9183
   extentZ=0.54
   shapeName=dyn_low_2_rope_r_box
   center=<1.3579 -0.1327 0.3018>
   rotation=<rot -0.0001 -2.8974 0>
  }]
 pack [
  {shapeType=Box, extentX=13.3291, extentY=0.6557, extentZ=9.7885, shapeName=pack_box, center=<0 -9.4081 1.0231>, rotation=<rot 0 0 51.8599>}
  {shapeType=Box, extentX=13.3291, extentY=1.6182, extentZ=9.8732, shapeName=pack_box, center=<0 -14.5032 -7.9581>, rotation=<rot 0 0 14.963>}
 ]
}
constraints [
 {bone1Name=dyn_low_1_rope_r, bone2Name=pack, swing1LimitDegrees=20, swing2LimitDegrees=45, twistLimitDegrees=45, twistMotion=Locked}
 {
  bone1Name=dyn_low_2_rope_r
  bone2Name=dyn_low_1_rope_r
  swing1LimitDegrees=20
  swing2LimitDegrees=45
  twistLimitDegrees=45
  twistMotion=Locked
 }
 {
  bone1Name=dyn_handle_1_joint
  bone2Name=pack
  swing1LimitDegrees=45
  swing1Motion=Locked
  swing2LimitDegrees=35
  twistLimitDegrees=45
  twistMotion=Locked
 }
]
