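# Animation Blueprint for the Bites backpack cosmetic, adding secondary physics motion to its rig.
# The input pose gets an additive pack-bone offset, then four chained AnimDynamics nodes swing rope bones with angular springs, per-node gravity and a shared Dyn_strenght alpha.
# A RigidBody node based at pack then simulates the pack's physics asset with variable-driven gravity before output.
# The handle AnimDynamics node is unconnected, and the event graph has no logic.
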AnimBlueprint /BRCosmetics/Accessories/FORT_Backpacks/Backpack_M_MED_Bites/Meshes/M_MED_Bites_Pack_AnimBP
target_skeleton=&/M_MED_Bites_Pack_Skeleton
graphs [
 AnimGraph {nodes [
   Anim Dynamics - Bone: dyn_handle_1_joint(Alpha=1, ExternalForce="0,0,0") {
    angularDampingOverride=0.999
    angularSpringConstant=20000
    bAngularSpring=true
    bOverrideAngularDamping=true
    bUseGravityOverride=true
    simulationSpace=BoneRelative
    alphaBoolBlend {blendOption=Linear}
    boundBone {boneName=dyn_handle_1_joint}
    componentAppliedLinearAccClamp=<3000 3000 3000>
    componentLinearAccScale=<0.3 0.3 0.3>
    componentLinearVelScale=<0.3 0.3 0.3>
    gravityOverride {z=-980}
    physicsBodyDefinitions [{
      boundBone {boneName=dyn_handle_1_joint}
      boxExtents=<2 6 6>
      constraintSetup {angularLimitsMax=<0 30 0>, angularLimitsMin=<0 -10 0>, angularTarget=<0 0 0>}
      localJointOffset=<-0.2 0 0>
     }]
   }
   n1 = "Transform (Modify) Bone - Bone: pack"(ComponentPose=Local To Component(LocalPose=Input Pose()), Translation="0,0,5", Rotation="0,0,0") {translationMode=BMM_Additive, alphaBoolBlend {blendOption=Linear}, boneToModify {boneName=pack}, translation {z=5}}
   n2 = Get Dyn_strenght() -> (Dyn_strenght:float)
   n3 = Anim Dynamics - Bone: dyn_rope_1_r(ComponentPose=@n1, Alpha=@n2, ExternalForce="0,0,0") {
    angularDampingOverride=1
    angularSpringConstant=20000
    bAngularSpring=true
    bOverrideAngularDamping=true
    bUseGravityOverride=true
    simulationSpace=BoneRelative
    alphaBoolBlend {blendOption=Linear}
    boundBone {boneName=dyn_rope_1_r}
    componentLinearAccScale=<0.8 0.8 0.8>
    componentLinearVelScale=<0.8 0.8 0.8>
    gravityOverride {z=-1900}
    physicsBodyDefinitions [{
      boundBone {boneName=dyn_rope_1_r}
      boxExtents=<3 3 3>
      constraintSetup {angularLimitsMax=<0 0 15>, angularLimitsMin=<0 -70 -15>, angularTarget=<0 0 0>}
      localJointOffset=<1 0 0>
     }]
    relativeSpaceBone {boneName=pack}
   }
   n4 = Anim Dynamics - Bone: dyn_rope_1_l(ComponentPose=@n3, Alpha=@n2, ExternalForce="0,0,0") {
    angularDampingOverride=1
    angularSpringConstant=20000
    bAngularSpring=true
    bOverrideAngularDamping=true
    bUseGravityOverride=true
    simulationSpace=BoneRelative
    alphaBoolBlend {blendOption=Linear}
    boundBone {boneName=dyn_rope_1_l}
    componentLinearAccScale=<0.8 0.8 0.85>
    componentLinearVelScale=<0.8 0.8 0.85>
    gravityOverride {z=-1100}
    physicsBodyDefinitions [{
      boundBone {boneName=dyn_rope_1_l}
      boxExtents=<3 3 3>
      constraintSetup {angularLimitsMax=<0 0 15>, angularLimitsMin=<0 -70 -15>, angularTarget=<0 0 0>}
      localJointOffset=<1 0 0>
     }]
    relativeSpaceBone {boneName=pack}
   }
   n5 = Anim Dynamics - Bone: dyn_rope_2_r(ComponentPose=@n4, Alpha=@n2, ExternalForce="0,0,0") {
    angularDampingOverride=1
    angularSpringConstant=20000
    bAngularSpring=true
    bOverrideAngularDamping=true
    bUseGravityOverride=true
    simulationSpace=BoneRelative
    alphaBoolBlend {blendOption=Linear}
    boundBone {boneName=dyn_rope_2_r}
    componentLinearAccScale=<0.8 0.8 0.8>
    componentLinearVelScale=<0.8 0.8 0.8>
    gravityOverride {z=-1300}
    physicsBodyDefinitions [{
      boundBone {boneName=dyn_rope_2_r}
      boxExtents=<2.5 2.5 2.5>
      constraintSetup {angularLimitsMax=<0 0 0>, angularLimitsMin=<0 -60 0>, angularTarget=<0 0 0>}
      localJointOffset=<0.5 0 0>
     }]
    relativeSpaceBone {boneName=pack}
   }
   n6 = Anim Dynamics - Bone: dyn_rope_2_l(ComponentPose=@n5, Alpha=@n2, ExternalForce="0,0,0") {
    angularDampingOverride=1
    angularSpringConstant=20000
    bAngularSpring=true
    bOverrideAngularDamping=true
    bUseGravityOverride=true
    simulationSpace=BoneRelative
    alphaBoolBlend {blendOption=Linear}
    boundBone {boneName=dyn_rope_2_l}
    componentLinearAccScale=<0.8 0.8 0.8>
    componentLinearVelScale=<0.8 0.8 0.8>
    gravityOverride {z=-1300}
    physicsBodyDefinitions [{
      boundBone {boneName=dyn_rope_2_l}
      boxExtents=<2.5 2.5 2.5>
      constraintSetup {angularLimitsMax=<0 0 0>, angularLimitsMin=<0 -60 0>, angularTarget=<0 0 0>}
      localJointOffset=<0.5 0 0>
     }]
    relativeSpaceBone {boneName=pack}
   }
   n7 = RigidBody!(ComponentPose=@n6, Alpha, OverrideWorldGravity:Vector=Get Override World Gravity() -> (Override World Gravity:Vector), ExternalForce) -> (Pose) {
    bClampLinearTranslationLimitToRefPose=true
    bOverrideWorldGravity=true
    overridePhysicsAsset=&/M_MED_Bites_Pack_Physics
    alphaBoolBlend {blendOption=Linear}
    baseBoneRef {boneName=pack}
    componentAppliedLinearAccClamp=<700 700 700>
    componentLinearAccScale=<1 1 1>
    componentLinearVelScale=<1 1 1>
    overrideWorldGravity {z=-980}
   }
   Output Pose(Result=Component To Local(ComponentPose=@n7)) {layerGroup=None}
  ]}
 EventGraph {nodes [
   TryGetPawnOwner()
   Event BlueprintUpdateAnimation()
  ]}
]
variables [
 Dyn_strenght {type=number, default=1}
 Override World Gravity {type=Vector, default=<0 0 -980>}
]
